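# Physics asset for the M_MED_TideCompost male medium player body, providing collision and jointed bodies for a two-sided towel.
# Capsules cover the upper spine, clavicles, arms and hands, with arm joints fully locked.
# Each towel chain locks base_towel_04 to dyn_towel_03, then swings through dyn_towel_04, 05 and a box-shaped 06.
# Swing and twist limits differ between left and right chains rather than mirroring.
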PhysicsAsset /BRCosmetics/Characters/Player/Male/Medium/Bodies/M_MED_TideCompost/Meshes/M_MED_TideCompost_Physics
bodies {
 base_towel_04_l [{shapeType=Capsule, length=0.505, radius=0.505, shapeName=base_towel_04_l_capsule, rotation=<rot 0 -90 0>}]
 base_towel_04_r [{shapeType=Capsule, length=0.505, radius=0.505, shapeName=base_towel_04_r_capsule, rotation=<rot 0 -90 -180>}]
 clavicle_l [{
   shapeType=Capsule
   length=9.1793
   radius=8
   shapeName=clavicle_l_capsule
   center=<5.5977 -1.7377 0.6614>
   rotation=<rot 64.0222 -2.2647 1.9547>
  }]
 clavicle_r [{
   shapeType=Capsule
   length=9.1793
   radius=8
   shapeName=clavicle_r_capsule
   center=<-5.5977 1.7377 -0.6614>
   rotation=<rot 64.0222 -2.2647 -178.0453>
  }]
 dyn_towel_03_l [{shapeType=Capsule, length=3.8351, radius=1, shapeName=dyn_towel_03_l_capsule, center=<1.9523 -0.5397 0.7675>, rotation=<rot 0 0 -116.6179>}]
 dyn_towel_03_r [{shapeType=Capsule, length=3.8351, radius=1, shapeName=dyn_towel_03_r_capsule, center=<-1.9523 0.5397 -0.7675>, rotation=<rot 0 0 63.3821>}]
 dyn_towel_04_l [{
   shapeType=Capsule
   length=6.1987
   radius=1.7348
   shapeName=dyn_towel_04_l_capsule
   center=<3.9693 0.2569 -0.3913>
   rotation=<rot -3.0996 -1.5973 -99.2297>
  }]
 dyn_towel_04_r [{
   shapeType=Capsule
   length=6.1987
   radius=1.7348
   shapeName=dyn_towel_04_r_capsule
   center=<-3.9969 -0.2113 -0.1413>
   rotation=<rot -3.0994 -1.5973 76.2565>
  }]
 dyn_towel_05_l [{
   shapeType=Capsule
   length=8.2488
   radius=1.9051
   shapeName=dyn_towel_05_l_capsule
   center=<3.2144 -0.2169 -0.5652>
   rotation=<rot -10.3443 0.2708 -104.3502>
  }]
 dyn_towel_05_r [{
   shapeType=Capsule
   length=9.1752
   radius=1.5164
   shapeName=dyn_towel_05_r_capsule
   center=<-3.2793 0.3084 0.2122>
   rotation=<rot -0.1028 -10.0512 79.399>
  }]
 dyn_towel_06_l [{
   shapeType=Box
   extentX=12.9858
   extentY=2.3531
   extentZ=5.1848
   shapeName=dyn_towel_06_l_box
   center=<5.0951 -0.8517 -1.4966>
   rotation=<rot 10.7732 83.9538 -90.2434>
  }]
 dyn_towel_06_r [{
   shapeType=Box
   extentX=11.5406
   extentY=1.632
   extentZ=9.0274
   shapeName=dyn_towel_06_r_box
   center=<-4.0337 1.3318 0.1319>
   rotation=<rot 14.3748 84.6635 89.8238>
  }]
 hand_l [{shapeType=Capsule, length=9.5087, radius=5.8695, center=<10.3629 -1.2652 0.0051>, rotation=<rot 0 82.2069 -91.9528>}]
 hand_r [{shapeType=Capsule, length=9.5087, radius=5.8695, center=<-10.3629 1.2652 -0.0051>, rotation=<rot 0 82.2069 88.0472>}]
 lowerarm_l [{shapeType=Capsule, length=15.1737, radius=5.5313, center=<10.3221 0.4244 0.6808>, rotation=<rot -0.0029 84.9497 -90.042>}]
 lowerarm_r [{shapeType=Capsule, length=15.1737, radius=5.5313, center=<-10.3221 -0.4244 -0.6808>, rotation=<rot -0.0029 84.9497 89.958>}]
 spine_04 [
  {
   shapeType=Capsule
   length=10.2596
   radius=11.3929
   shapeName=spine_04_capsule
   center=<3.0187 1.6159 0.0044>
   rotation=<rot 0.0001 -90 -0.0018>
  }
  {
   shapeType=Capsule
   length=9.5955
   radius=11.3929
   shapeName=spine_04_uppercapsule
   center=<6.9078 0.9905 -0.1253>
   rotation=<rot 8.3567 -90.0004 -0.0019>
  }
 ]
 spine_05 [{
   shapeType=Capsule
   length=14.8013
   radius=10.8087
   shapeName=spine_05_capsule
   center=<-1.3838 -1.5811 0.0275>
   rotation=<rot -1.0797 -90 1.6173>
  }]
 upperarm_l [
  {shapeType=Capsule, length=18.6219, radius=5.8565, center=<10.287 0.0296 -0.9589>, rotation=<rot 12.1519 92.7834 -91.2234>}
  {shapeType=Capsule, length=0, radius=8, shapeName=upperarm_l_capsule, center=<0.7327 -1.4442 -1.872>}
 ]
 upperarm_r [
  {shapeType=Capsule, length=18.6219, radius=5.8565, center=<-10.3971 1.5913 -0.1867>, rotation=<rot 12.1519 92.7834 88.7766>}
  {shapeType=Capsule, length=0, radius=8, shapeName=upperarm_r_capsule, center=<-0.7327 1.4442 1.872>, rotation=<rot 0 0 -180>}
 ]
}
constraints [
 {
  bone1Name=upperarm_l
  bone2Name=clavicle_l
  swing1LimitDegrees=45
  swing1Motion=Locked
  swing2LimitDegrees=45
  swing2Motion=Locked
  twistLimitDegrees=45
  twistMotion=Locked
 }
 {
  bone1Name=lowerarm_l
  bone2Name=upperarm_l
  swing1LimitDegrees=45
  swing1Motion=Locked
  swing2LimitDegrees=45
  swing2Motion=Locked
  twistLimitDegrees=45
  twistMotion=Locked
 }
 {
  bone1Name=hand_l
  bone2Name=lowerarm_l
  swing1LimitDegrees=45
  swing1Motion=Locked
  swing2LimitDegrees=45
  swing2Motion=Locked
  twistLimitDegrees=45
  twistMotion=Locked
 }
 {
  bone1Name=upperarm_r
  bone2Name=clavicle_r
  swing1LimitDegrees=45
  swing1Motion=Locked
  swing2LimitDegrees=45
  swing2Motion=Locked
  twistLimitDegrees=45
  twistMotion=Locked
 }
 {
  bone1Name=lowerarm_r
  bone2Name=upperarm_r
  swing1LimitDegrees=45
  swing1Motion=Locked
  swing2LimitDegrees=45
  swing2Motion=Locked
  twistLimitDegrees=45
  twistMotion=Locked
 }
 {
  bone1Name=hand_r
  bone2Name=lowerarm_r
  swing1LimitDegrees=45
  swing1Motion=Locked
  swing2LimitDegrees=45
  swing2Motion=Locked
  twistLimitDegrees=45
  twistMotion=Locked
 }
 {
  bone1Name=base_towel_04_l
  bone2Name=dyn_towel_03_l
  swing1LimitDegrees=45
  swing1Motion=Locked
  swing2LimitDegrees=45
  swing2Motion=Locked
  twistLimitDegrees=45
  twistMotion=Locked
 }
 {
  bone1Name=base_towel_04_r
  bone2Name=dyn_towel_03_r
  swing1LimitDegrees=45
  swing1Motion=Locked
  swing2LimitDegrees=45
  swing2Motion=Locked
  twistLimitDegrees=45
  twistMotion=Locked
 }
 {bone1Name=dyn_towel_04_l, bone2Name=base_towel_04_l, swing1LimitDegrees=20, swing2LimitDegrees=30, twistLimitDegrees=0}
 {bone1Name=dyn_towel_05_l, bone2Name=dyn_towel_04_l, swing1LimitDegrees=20, swing2LimitDegrees=30, twistLimitDegrees=0}
 {bone1Name=dyn_towel_04_r, bone2Name=base_towel_04_r, swing1LimitDegrees=10, swing2LimitDegrees=30, twistLimitDegrees=20}
 {bone1Name=dyn_towel_05_r, bone2Name=dyn_towel_04_r, swing1LimitDegrees=10, swing2LimitDegrees=30, twistLimitDegrees=20}
 {bone1Name=dyn_towel_06_l, bone2Name=dyn_towel_05_l, swing1LimitDegrees=20, swing2LimitDegrees=30, twistLimitDegrees=0}
 {bone1Name=dyn_towel_06_r, bone2Name=dyn_towel_05_r, swing1LimitDegrees=10, swing2LimitDegrees=30, twistLimitDegrees=20}
]
