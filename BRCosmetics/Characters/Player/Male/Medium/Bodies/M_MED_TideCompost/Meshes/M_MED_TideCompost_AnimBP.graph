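# Post-process style cosmetic AnimBP for the M_MED_TideCompost player body, adding secondary motion to a pose copied from the parent mesh.
# Control Rig feeds front-end or gameplay RigidBody sims, then string bone offsets, a crotch constraint and AnimDynamics strings with spherical limits.
# Muscle jiggle for arms, thighs and pecs follows, switched by bPbBlendEnabled.
# The event graph updates gravity/RBAN settings, movement-state values and Stage6-variant trunk-string dynamics.
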AnimBlueprint /BRCosmetics/Characters/Player/Male/Medium/Bodies/M_MED_TideCompost/Meshes/M_MED_TideCompost_AnimBP
target_skeleton=&/M_MED_TideCompost_Skeleton
graphs [
 AnimGraph {nodes [
   Layered blend per bone(BasePose=Use cached pose 'PreMuscleDynCache'(), BlendWeights_0=Get Custom Idle Is Playing Float() -> (Custom Idle Is Playing Float:double)) {layerSetup [{branchFilters [{boneName=dyn_shorts_in_01_l}, {boneName=dyn_shorts_in_02_l}, {boneName=dyn_shorts_in_01_r}, {boneName=dyn_shorts_in_02_r}]}]}
   Output Pose(Result=Blend Poses by bool(
     BlendPose_0=Use cached pose 'PreMuscleDynCache'()
     BlendPose_1=Use cached pose 'MuscleDynamicsCache'()
     bActiveValue=Get bPbBlendEnabled() -> (bPbBlendEnabled:bool)
    )) {layerGroup=None}
   n1 = RigidBody!(
    ComponentPose=Local To Component(LocalPose=Use cached pose 'ControlRigCache'())
    Alpha
    OverrideWorldGravity:Vector="0,0,-980"
    ExternalForce
    SimSpaceSettings:Sim Space Settings Structure=(WorldAlpha=1,VelocityScaleZ=1,DampingAlpha=0,MaxLinearVelocity=10,MaxAngularVelocity=800,MaxLinearAcceleration=40,MaxAngularAcceleration=4000,ExternalLinearDragV=(X=0.2,Y=0.2,Z=0.2),ExternalLinearVelocity=(X=0,Y=0,Z=0),ExternalAngularVelocity=(X=0,Y=0,Z=0))
   ) -> (Pose) {
    bOverrideWorldGravity=true
    overridePhysicsAsset=&/M_MED_TideCompost_Physics_FrontEnd
    simulationSpace=BaseBoneSpace
    baseBoneRef {boneName=spine_04}
    componentAppliedLinearAccClamp=<0 0 0>
    overrideWorldGravity {z=-980}
    simSpaceSettings {
     dampingAlpha=0
     maxAngularAcceleration=4000
     maxAngularVelocity=800
     maxLinearAcceleration=40
     maxLinearVelocity=10
     worldAlpha=1
     externalLinearDragV=<0.2 0.2 0.2>
    }
   }
   n2 = RigidBody(
    ComponentPose=Local To Component(LocalPose=Use cached pose 'ControlRigCache'())
    OverrideWorldGravity="0,0,-980"
    SimSpaceSettings=(WorldAlpha=1,VelocityScaleZ=1,DampingAlpha=0,MaxLinearVelocity=10,MaxAngularVelocity=800,MaxLinearAcceleration=40,MaxAngularAcceleration=4000,ExternalLinearDragV=(X=0.2,Y=0.2,Z=0.2),ExternalLinearVelocity=(X=0,Y=0,Z=0),ExternalAngularVelocity=(X=0,Y=0,Z=0))
   ) {
    bOverrideWorldGravity=true
    overridePhysicsAsset=&/M_MED_TideCompost_Physics
    simulationSpace=BaseBoneSpace
    baseBoneRef {boneName=spine_04}
    componentAppliedLinearAccClamp=<0 0 0>
    overrideWorldGravity {z=-980}
    simSpaceSettings {
     dampingAlpha=0
     maxAngularAcceleration=4000
     maxAngularVelocity=800
     maxLinearAcceleration=40
     maxLinearVelocity=10
     worldAlpha=1
     externalLinearDragV=<0.2 0.2 0.2>
    }
   }
   n3 = Blend Poses by bool(BlendPose_0=Component To Local(ComponentPose=@n1), BlendPose_1=Component To Local(ComponentPose=@n2), bActiveValue=Get bIsInFrontEnd())
   n4 = "Transform (Modify) Bone - Bone: dyn_string_02_r"(
    ComponentPose="Transform (Modify) Bone - Bone: dyn_string_02_l"(ComponentPose=Local To Component(LocalPose=@n3), Alpha=1, Translation="0,0,0", Rotation="5,0,0", Scale="1,1,1") {rotationMode=BMM_Additive, rotationSpace=BCS_BoneSpace, boneToModify {boneName=dyn_string_02_l}, rotation {pitch=5}}
    Alpha=1
    Translation="0,0,0"
    Rotation="4,0,0"
    Scale="1,1,1"
   ) {rotationMode=BMM_Additive, rotationSpace=BCS_BoneSpace, boneToModify {boneName=dyn_string_02_r}, rotation {pitch=4}}
   n5 = "Transform (Modify) Bone - Bone: dyn_string_02_r"(
    ComponentPose="Transform (Modify) Bone - Bone: dyn_string_02_l"(
     ComponentPose="Transform (Modify) Bone - Bone: base_string"(ComponentPose=@n4, Alpha=1, Translation="0,0,0", Rotation="0,0,-15", Scale="1,1,1") {rotationMode=BMM_Additive, rotationSpace=BCS_BoneSpace, boneToModify {boneName=base_string}, rotation {roll=-15}}
     Alpha=1
     Translation="0,0,0"
     Rotation="8,0,0"
     Scale="1,1,1"
    ) {rotationMode=BMM_Additive, rotationSpace=BCS_BoneSpace, boneToModify {boneName=dyn_string_02_l}, rotation {pitch=8}}
    Alpha=1
    Translation="0,0,0"
    Rotation="12,0,0"
    Scale="1,1,1"
   ) {rotationMode=BMM_Additive, rotationSpace=BCS_BoneSpace, boneToModify {boneName=dyn_string_02_r}, rotation {pitch=12}}
   Save cached pose 'PreAnimDyn'(Pose=Component To Local(ComponentPose=@n5)) {cacheName=PreAnimDyn}
   n6 = Anim Dynamics - Bone: dyn_string_01_r(ComponentPose=Local To Component(LocalPose=Use cached pose 'PreAnimDyn'()), Alpha=1, ExternalForce="0,0,0") {
    angularDampingOverride=1
    angularSpringConstant=20000
    bAngularSpring=true
    bOverrideAngularDamping=true
    bUseGravityOverride=true
    gravityScale=0.35
    simulationSpace=RootRelative
    boundBone {boneName=dyn_string_01_r}
    componentLinearAccScale=<1 1 1>
    componentLinearVelScale=<1 1 1>
    gravityOverride {z=-980}
    physicsBodyDefinitions [{
      boundBone {boneName=dyn_string_01_r}
      boxExtents=<20 30 30>
      constraintSetup {angularLimitsMax=<0 0 15>, angularLimitsMin=<0 -10 -15>, angularTarget=<1 0 0>}
      localJointOffset=<4 0 0>
     }]
    simSpaceSettings {maxAngularAcceleration=20, maxAngularVelocity=2, simSpaceAngularAlpha=1}
   }
   n7 = Anim Dynamics - Bone: dyn_string_01_l(ComponentPose=@n6, Alpha=1, ExternalForce="0,0,0") {
    angularDampingOverride=1
    angularSpringConstant=20000
    bAngularSpring=true
    bOverrideAngularDamping=true
    bUseGravityOverride=true
    gravityScale=0.35
    simulationSpace=RootRelative
    boundBone {boneName=dyn_string_01_l}
    componentLinearAccScale=<1 1 1>
    componentLinearVelScale=<1 1 1>
    gravityOverride {z=-980}
    physicsBodyDefinitions [{
      boundBone {boneName=dyn_string_01_l}
      boxExtents=<20 30 30>
      constraintSetup {angularLimitsMax=<0 0 15>, angularLimitsMin=<0 -10 -15>, angularTarget=<1 0 0>}
      localJointOffset=<4 0 0>
     }]
    simSpaceSettings {maxAngularAcceleration=20, maxAngularVelocity=2, simSpaceAngularAlpha=1}
   }
   n8 = Component To Local(ComponentPose=@n7)
   n9 = Use cached pose 'PreAnimDyn'()
   n10 = Get Trunk String ON() -> (Trunk String ON:bool)
   n11 = "Constraint - deform_crotch_base (PPPPP)"!(
    ComponentPose=Local To Component(LocalPose=Blend Poses by bool(BlendPose_0=@n8, BlendPose_1=@n9, bActiveValue=@n10))
    Alpha=1
    ConstraintWeights_0=0.4206
    ConstraintWeights_1:float=0.1537
    ConstraintWeights_2:float=0.1486
    ConstraintWeights_3:float=0.1462
    ConstraintWeights_4:float=0.1308
   ) -> (Pose) {
    boneToModify {boneName=deform_crotch_base}
    constraintSetup [
     {targetBone {boneName=pelvis}}
     {targetBone {boneName=deform_groin_l}}
     {targetBone {boneName=thigh_twist_01_l}}
     {targetBone {boneName=thigh_twist_01_r}}
     {targetBone {boneName=deform_groin_r}}
    ]
    constraintWeights [1,1,1,1,1]
   }
   n12 = Anim Dynamics - Bone: dyn_string_02_r(ComponentPose=@n11, Alpha=1, ExternalForce="0,0,0") {
    angularDampingOverride=1
    angularSpringConstant=10000
    bAngularSpring=true
    bOverrideAngularDamping=true
    bUseGravityOverride=true
    bUsePlanarLimit=false
    bUseSphericalLimits=true
    gravityScale=0.35
    simulationSpace=RootRelative
    boundBone {boneName=dyn_string_02_r}
    componentAppliedLinearAccClamp=<3000 3000 3000>
    componentLinearAccScale=<0.5 0.5 0.5>
    componentLinearVelScale=<0.5 0.5 0.5>
    gravityOverride {z=-980}
    physicsBodyDefinitions [{
      collisionType=CustomSphere
      sphereCollisionRadius=1.4
      boundBone {boneName=dyn_string_02_r}
      boxExtents=<20 30 30>
      constraintSetup {angularLimitsMax=<0 0 15>, angularLimitsMin=<0 -25 -15>, angularTarget=<1 0 0>}
      localJointOffset=<4 0 0>
     }]
    simSpaceSettings {maxAngularAcceleration=40, maxAngularVelocity=4, simSpaceAngularAlpha=1}
    sphericalLimits [
     {limitRadius=19, drivingBone {boneName=deform_crotch_dyn}, sphereLocalOffset=<0 -19.5 2>}
     {limitRadius=11.6, drivingBone {boneName=thigh_r}, sphereLocalOffset=<13 5 0>}
    ]
   }
   n13 = Anim Dynamics - Bone: dyn_string_02_l(ComponentPose=@n12, Alpha=1, ExternalForce="0,0,0") {
    angularDampingOverride=1
    angularSpringConstant=10000
    bAngularSpring=true
    bOverrideAngularDamping=true
    bUseGravityOverride=true
    bUsePlanarLimit=false
    bUseSphericalLimits=true
    gravityScale=0.35
    simulationSpace=RootRelative
    boundBone {boneName=dyn_string_02_l}
    componentAppliedLinearAccClamp=<3000 3000 3000>
    componentLinearAccScale=<0.5 0.5 0.5>
    componentLinearVelScale=<0.5 0.5 0.5>
    gravityOverride {z=-980}
    physicsBodyDefinitions [{
      collisionType=CustomSphere
      sphereCollisionRadius=1.4
      boundBone {boneName=dyn_string_02_l}
      boxExtents=<20 30 30>
      constraintSetup {angularLimitsMax=<0 0 15>, angularLimitsMin=<0 -25 -15>, angularTarget=<1 0 0>}
      localJointOffset=<4 0 0>
     }]
    simSpaceSettings {maxAngularAcceleration=40, maxAngularVelocity=4, simSpaceAngularAlpha=1}
    sphericalLimits [{limitRadius=19, drivingBone {boneName=deform_crotch_dyn}, sphereLocalOffset=<0 -19.5 2>}]
   }
   n14 = Anim Dynamics - Bone: dyn_string_03_r(ComponentPose=@n13, Alpha=1, ExternalForce="0,0,0") {
    angularDampingOverride=1
    angularSpringConstant=10000
    bAngularSpring=true
    bOverrideAngularDamping=true
    bUseGravityOverride=true
    bUsePlanarLimit=false
    bUseSphericalLimits=true
    gravityScale=0.35
    simulationSpace=RootRelative
    boundBone {boneName=dyn_string_03_r}
    componentAppliedLinearAccClamp=<3000 3000 3000>
    componentLinearAccScale=<0.5 0.5 0.5>
    componentLinearVelScale=<0.5 0.5 0.5>
    gravityOverride {z=-980}
    physicsBodyDefinitions [{
      collisionType=CustomSphere
      sphereCollisionRadius=1.4
      boundBone {boneName=dyn_string_03_r}
      boxExtents=<20 30 30>
      constraintSetup {angularLimitsMax=<0 0 15>, angularLimitsMin=<0 -25 -15>, angularTarget=<1 0 0>}
      localJointOffset=<4 0 0>
     }]
    simSpaceSettings {maxAngularAcceleration=40, maxAngularVelocity=4, simSpaceAngularAlpha=1}
    sphericalLimits [
     {limitRadius=19, drivingBone {boneName=deform_crotch_dyn}, sphereLocalOffset=<0 -19.5 2>}
     {limitRadius=11.6, drivingBone {boneName=thigh_r}, sphereLocalOffset=<13 5 0>}
    ]
   }
   n15 = Anim Dynamics - Bone: dyn_string_03_l(ComponentPose=@n14, Alpha=1, ExternalForce="0,0,0") {
    angularDampingOverride=1
    angularSpringConstant=10000
    bAngularSpring=true
    bOverrideAngularDamping=true
    bUseGravityOverride=true
    bUsePlanarLimit=false
    bUseSphericalLimits=true
    gravityScale=0.35
    simulationSpace=RootRelative
    boundBone {boneName=dyn_string_03_l}
    componentAppliedLinearAccClamp=<3000 3000 3000>
    componentLinearAccScale=<0.5 0.5 0.5>
    componentLinearVelScale=<0.5 0.5 0.5>
    gravityOverride {z=-980}
    physicsBodyDefinitions [{
      collisionType=CustomSphere
      sphereCollisionRadius=1.4
      boundBone {boneName=dyn_string_03_l}
      boxExtents=<20 30 30>
      constraintSetup {angularLimitsMax=<0 0 15>, angularLimitsMin=<0 -25 -15>, angularTarget=<1 0 0>}
      localJointOffset=<4 0 0>
     }]
    simSpaceSettings {maxAngularAcceleration=40, maxAngularVelocity=4, simSpaceAngularAlpha=1}
    sphericalLimits [{limitRadius=19, drivingBone {boneName=deform_crotch_dyn}, sphereLocalOffset=<0 -19.5 2>}]
   }
   Save cached pose 'PreMuscleDynCache'(Pose=Component To Local(ComponentPose=@n15)) {cacheName=PreMuscleDynCache}
   n16 = Control Rig(Source=Copy Pose From Mesh() {bUseAttachedParent=true}) {controlRigAssetReference {controlRigAsset=&/M_MED_TideCompost_CtrlRig}}
   Save cached pose 'ControlRigCache'(Pose=@n16) {cacheName=ControlRigCache}
   n17 = Anim Dynamics - Bone: deform_bicep_l(ComponentPose=Local To Component(LocalPose=Use cached pose 'PreMuscleDynCache'()), Alpha=1, ExternalForce="0,0,0") {
    angularSpringConstant=15000
    bLinearSpring=true
    bOverrideLinearDamping=true
    bUseGravityOverride=true
    bUsePlanarLimit=false
    linearDampingOverride=1
    linearSpringConstant=500
    simulationSpace=RootRelative
    boundBone {boneName=deform_bicep_l}
    componentLinearAccScale=<0.2 0.2 0.2>
    componentLinearVelScale=<0.2 0.2 0.2>
    gravityOverride {z=-200}
    physicsBodyDefinitions [{
      boundBone {boneName=deform_bicep_l}
      boxExtents=<18 18 18>
      constraintSetup {angularLimitsMax=<0 0 0>, angularLimitsMin=<0 0 0>, angularTarget=<1 0 0>, linearAxesMax=<0 0 0.5>, linearAxesMin=<0 0 -0.5>}
      localJointOffset=<-2 0 0>
     }]
    relativeSpaceBone {boneName=upperarm_l}
    simSpaceSettings {maxAngularAcceleration=40, maxAngularVelocity=770, simSpaceAngularAlpha=1}
   }
   n18 = Anim Dynamics - Bone: deform_bicep_r(ComponentPose=@n17, Alpha=1, ExternalForce="0,0,0") {
    angularDampingOverride=0.9999
    angularSpringConstant=15000
    bAngularSpring=true
    bOverrideAngularDamping=true
    bUseGravityOverride=true
    bUsePlanarLimit=false
    simulationSpace=RootRelative
    boundBone {boneName=deform_bicep_r}
    componentLinearAccScale=<0.2 0.2 0.2>
    componentLinearVelScale=<0.2 0.2 0.2>
    gravityOverride {z=-980}
    physicsBodyDefinitions [{
      boundBone {boneName=deform_bicep_r}
      boxExtents=<6 18 6>
      constraintSetup {angularLimitsMax=<1 10 1>, angularLimitsMin=<-1 -10 -1>, angularTarget=<1 0 0>}
      localJointOffset=<-5 0 0>
     }]
    relativeSpaceBone {boneName=upperarm_l}
    simSpaceSettings {maxAngularAcceleration=10, maxAngularVelocity=300, simSpaceAngularAlpha=1}
   }
   n19 = Anim Dynamics - Bone: deform_bicep_l(ComponentPose=@n18, Alpha=1, ExternalForce="0,0,0") {
    angularSpringConstant=15000
    bLinearSpring=true
    bOverrideLinearDamping=true
    bUseGravityOverride=true
    bUsePlanarLimit=false
    linearDampingOverride=1
    linearSpringConstant=500
    simulationSpace=RootRelative
    boundBone {boneName=deform_bicep_l}
    componentLinearAccScale=<0.2 0.2 0.2>
    componentLinearVelScale=<0.2 0.2 0.2>
    gravityOverride {z=-200}
    physicsBodyDefinitions [{
      boundBone {boneName=deform_bicep_l}
      boxExtents=<18 18 18>
      constraintSetup {angularLimitsMax=<0 0 0>, angularLimitsMin=<0 0 0>, angularTarget=<1 0 0>, linearAxesMax=<0 0 0.5>, linearAxesMin=<0 0 -0.5>}
      localJointOffset=<2 0 0>
     }]
    relativeSpaceBone {boneName=upperarm_r}
    simSpaceSettings {maxAngularAcceleration=40, maxAngularVelocity=770, simSpaceAngularAlpha=1}
   }
   n20 = Anim Dynamics - Bone: deform_bicep_r(ComponentPose=@n19, Alpha=1, ExternalForce="0,0,0") {
    angularDampingOverride=0.9999
    angularSpringConstant=15000
    bAngularSpring=true
    bOverrideAngularDamping=true
    bUseGravityOverride=true
    bUsePlanarLimit=false
    simulationSpace=RootRelative
    boundBone {boneName=deform_bicep_r}
    componentLinearAccScale=<0.2 0.2 0.2>
    componentLinearVelScale=<0.2 0.2 0.2>
    gravityOverride {z=-980}
    physicsBodyDefinitions [{
      boundBone {boneName=deform_bicep_r}
      boxExtents=<6 18 6>
      constraintSetup {angularLimitsMax=<1 10 1>, angularLimitsMin=<-1 -10 -1>, angularTarget=<1 0 0>}
      localJointOffset=<-5 0 0>
     }]
    relativeSpaceBone {boneName=upperarm_r}
    simSpaceSettings {maxAngularAcceleration=10, maxAngularVelocity=300, simSpaceAngularAlpha=1}
   }
   n21 = Anim Dynamics - Bone: deform_tricep_l(ComponentPose=@n20, Alpha=1, ExternalForce="0,0,0") {
    angularSpringConstant=15000
    bLinearSpring=true
    bOverrideLinearDamping=true
    bUseGravityOverride=true
    bUsePlanarLimit=false
    linearDampingOverride=1
    linearSpringConstant=450
    simulationSpace=RootRelative
    boundBone {boneName=deform_tricep_l}
    componentLinearAccScale=<0.15 0.15 0.15>
    componentLinearVelScale=<0.15 0.15 0.15>
    gravityOverride {z=-100}
    physicsBodyDefinitions [{
      boundBone {boneName=deform_tricep_l}
      boxExtents=<30 40 30>
      constraintSetup {angularLimitsMax=<0 0 0>, angularLimitsMin=<0 0 0>, angularTarget=<1 0 0>, linearAxesMax=<0 0.1 0.75>, linearAxesMin=<0 -0.1 -0.75>}
      localJointOffset=<-3 0 0>
     }]
    relativeSpaceBone {boneName=upperarm_l}
    simSpaceSettings {maxAngularAcceleration=30, maxAngularVelocity=300, simSpaceAngularAlpha=1}
   }
   n22 = Anim Dynamics - Bone: deform_tricep_l(ComponentPose=@n21, Alpha=1, ExternalForce="0,0,0") {
    angularDampingOverride=0.9999
    angularSpringConstant=50000
    bAngularSpring=true
    bOverrideAngularDamping=true
    bUseGravityOverride=true
    bUsePlanarLimit=false
    simulationSpace=BoneRelative
    boundBone {boneName=deform_tricep_l}
    componentLinearAccScale=<0.15 0.15 0.15>
    componentLinearVelScale=<0.15 0.15 0.15>
    gravityOverride {z=-980}
    physicsBodyDefinitions [{
      boundBone {boneName=deform_tricep_l}
      boxExtents=<12 20 12>
      constraintSetup {angularLimitsMax=<1 10 1>, angularLimitsMin=<-1 -10 -1>, angularTarget=<1 0 0>}
      localJointOffset=<5 0 0>
     }]
    relativeSpaceBone {boneName=upperarm_l}
    simSpaceSettings {maxAngularAcceleration=10, maxAngularVelocity=300, simSpaceAngularAlpha=1}
   }
   n23 = Anim Dynamics - Bone: deform_tricep_r(ComponentPose=@n22, Alpha=1, ExternalForce="0,0,0") {
    angularSpringConstant=15000
    bLinearSpring=true
    bOverrideLinearDamping=true
    bUseGravityOverride=true
    bUsePlanarLimit=false
    linearDampingOverride=1
    linearSpringConstant=450
    simulationSpace=RootRelative
    boundBone {boneName=deform_tricep_r}
    componentLinearAccScale=<0.15 0.15 0.15>
    componentLinearVelScale=<0.15 0.15 0.15>
    gravityOverride {z=-100}
    physicsBodyDefinitions [{
      boundBone {boneName=deform_tricep_r}
      boxExtents=<30 40 30>
      constraintSetup {angularLimitsMax=<0 0 0>, angularLimitsMin=<0 0 0>, angularTarget=<1 0 0>, linearAxesMax=<0 0.1 0.75>, linearAxesMin=<0 -0.1 -0.75>}
      localJointOffset=<3 0 0>
     }]
    relativeSpaceBone {boneName=upperarm_r}
    simSpaceSettings {maxAngularAcceleration=30, maxAngularVelocity=300, simSpaceAngularAlpha=1}
   }
   n24 = Anim Dynamics - Bone: deform_tricep_r(ComponentPose=@n23, Alpha=1, ExternalForce="0,0,0") {
    angularDampingOverride=0.9999
    angularSpringConstant=50000
    bAngularSpring=true
    bOverrideAngularDamping=true
    bUseGravityOverride=true
    bUsePlanarLimit=false
    simulationSpace=BoneRelative
    boundBone {boneName=deform_tricep_r}
    componentLinearAccScale=<0.15 0.15 0.15>
    componentLinearVelScale=<0.15 0.15 0.15>
    gravityOverride {z=-980}
    physicsBodyDefinitions [{
      boundBone {boneName=deform_tricep_r}
      boxExtents=<12 20 12>
      constraintSetup {angularLimitsMax=<1 10 1>, angularLimitsMin=<-1 -10 -1>, angularTarget=<1 0 0>}
      localJointOffset=<5 0 0>
     }]
    relativeSpaceBone {boneName=upperarm_r}
    simSpaceSettings {maxAngularAcceleration=10, maxAngularVelocity=300, simSpaceAngularAlpha=1}
   }
   n25 = Anim Dynamics - Bone: thigh_quads_l(ComponentPose=@n24, Alpha=1, ExternalForce="0,0,0") {
    angularSpringConstant=15000
    bLinearSpring=true
    bOverrideLinearDamping=true
    bUseGravityOverride=true
    bUsePlanarLimit=false
    linearDampingOverride=1
    linearSpringConstant=350
    simulationSpace=RootRelative
    boundBone {boneName=thigh_quads_l}
    componentLinearAccScale=<0.42 0.42 0.42>
    componentLinearVelScale=<0.42 0.42 0.42>
    gravityOverride {z=-380}
    physicsBodyDefinitions [{
      boundBone {boneName=thigh_quads_l}
      boxExtents=<8 8 8>
      constraintSetup {angularLimitsMax=<0 0 0>, angularLimitsMin=<0 0 0>, angularTarget=<1 0 0>, linearAxesMax=<0 0 0.5>, linearAxesMin=<0 0 -0.5>}
      localJointOffset=<-2 0 0>
     }]
    relativeSpaceBone {boneName=upperarm_l}
    simSpaceSettings {maxAngularAcceleration=40, maxAngularVelocity=770, simSpaceAngularAlpha=1}
   }
   n26 = Anim Dynamics - Bone: thigh_quads_l(ComponentPose=@n25, Alpha=1, ExternalForce="0,0,0") {
    angularDampingOverride=0.9999
    angularSpringConstant=10000
    bAngularSpring=true
    bOverrideAngularDamping=true
    bUseGravityOverride=true
    bUsePlanarLimit=false
    simulationSpace=RootRelative
    boundBone {boneName=thigh_quads_l}
    componentLinearAccScale=<0.42 0.42 0.42>
    componentLinearVelScale=<0.42 0.42 0.42>
    gravityOverride {z=-980}
    physicsBodyDefinitions [{
      boundBone {boneName=thigh_quads_l}
      boxExtents=<6 11 6>
      constraintSetup {angularLimitsMax=<1 10 1>, angularLimitsMin=<-1 -10 -1>, angularTarget=<1 0 0>}
      localJointOffset=<-5 0 0>
     }]
    relativeSpaceBone {boneName=upperarm_l}
    simSpaceSettings {maxAngularAcceleration=10, maxAngularVelocity=300, simSpaceAngularAlpha=1}
   }
   n27 = Anim Dynamics - Bone: thigh_hamstring_l(ComponentPose=@n26, Alpha=1, ExternalForce="0,0,0") {
    angularSpringConstant=15000
    bLinearSpring=true
    bOverrideLinearDamping=true
    bUseGravityOverride=true
    bUsePlanarLimit=false
    linearDampingOverride=1
    linearSpringConstant=350
    simulationSpace=RootRelative
    boundBone {boneName=thigh_hamstring_l}
    componentLinearAccScale=<0.42 0.42 0.42>
    componentLinearVelScale=<0.42 0.42 0.42>
    gravityOverride {z=-380}
    physicsBodyDefinitions [{
      boundBone {boneName=thigh_hamstring_l}
      boxExtents=<8 8 8>
      constraintSetup {angularLimitsMax=<0 0 0>, angularLimitsMin=<0 0 0>, angularTarget=<1 0 0>, linearAxesMax=<0 0 0.5>, linearAxesMin=<0 0 -0.5>}
      localJointOffset=<2 0 0>
     }]
    relativeSpaceBone {boneName=upperarm_r}
    simSpaceSettings {maxAngularAcceleration=40, maxAngularVelocity=770, simSpaceAngularAlpha=1}
   }
   n28 = Anim Dynamics - Bone: thigh_hamstring_l(ComponentPose=@n27, Alpha=1, ExternalForce="0,0,0") {
    angularDampingOverride=0.9999
    angularSpringConstant=10000
    bAngularSpring=true
    bOverrideAngularDamping=true
    bUseGravityOverride=true
    bUsePlanarLimit=false
    simulationSpace=RootRelative
    boundBone {boneName=thigh_hamstring_l}
    componentLinearAccScale=<0.42 0.42 0.42>
    componentLinearVelScale=<0.42 0.42 0.42>
    gravityOverride {z=-980}
    physicsBodyDefinitions [{
      boundBone {boneName=thigh_hamstring_l}
      boxExtents=<6 11 6>
      constraintSetup {angularLimitsMax=<1 10 1>, angularLimitsMin=<-1 -10 -1>, angularTarget=<1 0 0>}
      localJointOffset=<-5 0 0>
     }]
    relativeSpaceBone {boneName=upperarm_r}
    simSpaceSettings {maxAngularAcceleration=10, maxAngularVelocity=300, simSpaceAngularAlpha=1}
   }
   n29 = Anim Dynamics - Bone: thigh_quads_r(ComponentPose=@n28, Alpha=1, ExternalForce="0,0,0") {
    angularSpringConstant=15000
    bLinearSpring=true
    bOverrideLinearDamping=true
    bUseGravityOverride=true
    bUsePlanarLimit=false
    linearDampingOverride=1
    linearSpringConstant=350
    simulationSpace=RootRelative
    boundBone {boneName=thigh_quads_r}
    componentLinearAccScale=<0.42 0.42 0.42>
    componentLinearVelScale=<0.42 0.42 0.42>
    gravityOverride {z=-380}
    physicsBodyDefinitions [{
      boundBone {boneName=thigh_quads_r}
      boxExtents=<6 6 6>
      constraintSetup {angularLimitsMax=<0 0 0>, angularLimitsMin=<0 0 0>, angularTarget=<1 0 0>, linearAxesMax=<0 0 0.5>, linearAxesMin=<0 0 -0.5>}
      localJointOffset=<-2 0 0>
     }]
    relativeSpaceBone {boneName=upperarm_l}
    simSpaceSettings {maxAngularAcceleration=40, maxAngularVelocity=770, simSpaceAngularAlpha=1}
   }
   n30 = Anim Dynamics - Bone: thigh_quads_r(ComponentPose=@n29, Alpha=1, ExternalForce="0,0,0") {
    angularDampingOverride=0.9999
    angularSpringConstant=10000
    bAngularSpring=true
    bOverrideAngularDamping=true
    bUseGravityOverride=true
    bUsePlanarLimit=false
    simulationSpace=RootRelative
    boundBone {boneName=thigh_quads_r}
    componentLinearAccScale=<0.4 0.4 0.4>
    componentLinearVelScale=<0.4 0.4 0.4>
    gravityOverride {z=-980}
    physicsBodyDefinitions [{
      boundBone {boneName=thigh_quads_r}
      boxExtents=<5 8 5>
      constraintSetup {angularLimitsMax=<1 10 1>, angularLimitsMin=<-1 -10 -1>, angularTarget=<1 0 0>}
      localJointOffset=<-5 0 0>
     }]
    relativeSpaceBone {boneName=upperarm_l}
    simSpaceSettings {maxAngularAcceleration=10, maxAngularVelocity=300, simSpaceAngularAlpha=1}
   }
   n31 = Anim Dynamics - Bone: thigh_hamstring_r(ComponentPose=@n30, Alpha=1, ExternalForce="0,0,0") {
    angularSpringConstant=15000
    bLinearSpring=true
    bOverrideLinearDamping=true
    bUseGravityOverride=true
    bUsePlanarLimit=false
    linearDampingOverride=1
    linearSpringConstant=350
    simulationSpace=RootRelative
    boundBone {boneName=thigh_hamstring_r}
    componentLinearAccScale=<0.42 0.42 0.42>
    componentLinearVelScale=<0.42 0.42 0.42>
    gravityOverride {z=-380}
    physicsBodyDefinitions [{
      boundBone {boneName=thigh_hamstring_r}
      boxExtents=<6 6 6>
      constraintSetup {angularLimitsMax=<0 0 0>, angularLimitsMin=<0 0 0>, angularTarget=<1 0 0>, linearAxesMax=<0 0 0.5>, linearAxesMin=<0 0 -0.5>}
      localJointOffset=<2 0 0>
     }]
    relativeSpaceBone {boneName=upperarm_r}
    simSpaceSettings {maxAngularAcceleration=40, maxAngularVelocity=770, simSpaceAngularAlpha=1}
   }
   n32 = Anim Dynamics - Bone: thigh_hamstring_r(ComponentPose=@n31, Alpha=1, ExternalForce="0,0,0") {
    angularDampingOverride=0.9999
    angularSpringConstant=10000
    bAngularSpring=true
    bOverrideAngularDamping=true
    bUseGravityOverride=true
    bUsePlanarLimit=false
    simulationSpace=RootRelative
    boundBone {boneName=thigh_hamstring_r}
    componentLinearAccScale=<0.4 0.4 0.4>
    componentLinearVelScale=<0.4 0.4 0.4>
    gravityOverride {z=-980}
    physicsBodyDefinitions [{
      boundBone {boneName=thigh_hamstring_r}
      boxExtents=<5 8 5>
      constraintSetup {angularLimitsMax=<1 10 1>, angularLimitsMin=<-1 -10 -1>, angularTarget=<1 0 0>}
      localJointOffset=<-5 0 0>
     }]
    relativeSpaceBone {boneName=upperarm_r}
    simSpaceSettings {maxAngularAcceleration=10, maxAngularVelocity=300, simSpaceAngularAlpha=1}
   }
   n33 = Anim Dynamics - Bone: deform_pec_l!(ComponentPose=@n32, ExternalForce="0,0,0") -> (Pose) {
    alpha=0.8
    angularSpringConstant=15000
    bLinearSpring=true
    bOverrideLinearDamping=true
    bUseGravityOverride=true
    bUsePlanarLimit=false
    linearDampingOverride=1
    linearSpringConstant=100
    simulationSpace=RootRelative
    boundBone {boneName=deform_pec_l}
    componentAppliedLinearAccClamp=<1000 1000 1000>
    gravityOverride {z=-400}
    physicsBodyDefinitions [{
      boundBone {boneName=deform_pec_l}
      boxExtents=<70 80 90>
      constraintSetup {angularLimitsMax=<-2 0 0>, angularLimitsMin=<-15 0 0>, angularTarget=<1 0 0>, linearAxesMax=<1 0 3>, linearAxesMin=<-1 0 -3>}
      localJointOffset=<-4 0 0>
     }]
    relativeSpaceBone {boneName=pelvis}
    simSpaceSettings {maxAngularAcceleration=10, maxAngularVelocity=770, simSpaceAngularAlpha=0.4}
   }
   n34 = Anim Dynamics - Bone: deform_pec_r!(ComponentPose=@n33, ExternalForce="0,0,0") -> (Pose) {
    alpha=0.8
    angularSpringConstant=15000
    bLinearSpring=true
    bOverrideLinearDamping=true
    bUseGravityOverride=true
    bUsePlanarLimit=false
    linearDampingOverride=1
    linearSpringConstant=100
    simulationSpace=RootRelative
    boundBone {boneName=deform_pec_r}
    componentAppliedLinearAccClamp=<1000 1000 1000>
    gravityOverride {z=-400}
    physicsBodyDefinitions [{
      boundBone {boneName=deform_pec_r}
      boxExtents=<70 80 90>
      constraintSetup {angularLimitsMax=<0 0 0>, angularLimitsMin=<0 0 0>, angularTarget=<1 0 0>, linearAxesMax=<1 0 3>, linearAxesMin=<-1 0 -3>}
      localJointOffset=<4 0 0>
     }]
    relativeSpaceBone {boneName=pelvis}
    simSpaceSettings {maxAngularAcceleration=10, maxAngularVelocity=770, simSpaceAngularAlpha=0.4}
   }
   Save cached pose 'MuscleDynamicsCache'(Pose=Component To Local(ComponentPose=@n34)) {cacheName=MuscleDynamicsCache}
  ]}
 EventGraph {nodes [
   Event BlueprintUpdateAnimation()
   > n1 = Sequence!(execute) -> (then_0, then_1, then_2:Exec, then_3:Exec, then_4:Exec, then_5:Exec)
   > n2 = "Gravity Override State Function (Parameter Asset) (Nativized)"(
    CharacterPartAnimInstance=Self-Reference()
    GameTimeInSeconds=GetGameTimeInSeconds()
    Parameters=&/M_MED_TideCompost_Parameters
   )
   > Set Gravity Override(Gravity Override=@n2)
   > Set RBAN Settings Out(RBAN Settings Out=@n2.OutRigidBodyAnimNodeParameters)
   > n3 = Branch(Condition=CharacterItemDefinitionHasAnyMetaTag(Tags=Make Array("[0]"=(TagName="Cosmetics.Variant.Property.Stage6"))))
   > Set Trunk String ON(Trunk String ON:bool=true) -> (Output_Get:bool)
   Set Trunk String ON(execute=@n3.else, Trunk String ON=false)
   n4 = NOT Boolean(A=Get bIsSliding())
   n5 = Get bIsCrouching()
   Set OnlyCrouching(execute=@n1.then_3, OnlyCrouching:double="To Float (Boolean)"(InBool=AND Boolean(A=@n5, B=@n4))) -> (Output_Get:double)
   n6 = Get bIsParachuteOpened()
   n7 = Get bIsSkydiveFloating()
   n8 = Get bIsSkyDiving()
   n9 = Branch(execute=@n1.then_1, Condition=AND Boolean(A=OR Boolean(A=@n8, B=@n7), B=NOT Boolean(A=@n6)))
   > "Set SkyDiving "("SkyDiving ":double=1) -> (Output_Get:double)
   "Set SkyDiving "(execute=@n9.else, "SkyDiving "=0)
   Branch(execute=@n1.then_5, Condition=Get bIsInFrontEnd())
   > Set Custom Idle Is Playing Float(Custom Idle Is Playing Float:double=GetCurveValue(self=GetOwnerFrontEndAnimInstance(), CurveName=Frontend_Idle_Bypass)) -> (Output_Get:double)
   n10 = NOT Boolean(A=Get bIsParachuteOpened())
   n11 = OR Boolean(A=Get bIsJumping(), B=Get bIsFalling())
   Set is_Jumping(execute=@n1.then_2, is_Jumping:double="To Float (Boolean)"(InBool=AND Boolean(A=@n11, B=@n10))) -> (Output_Get:double)
   n12 = Branch(execute=@n1.then_4, Condition=Get bIsPlayingEmote())
   > Set SimlationSpace(SimlationSpace:AnimPhysSimSpaceType Enum) -> (Output_Get:AnimPhysSimSpaceType Enum)
   > Set AlphaPlayingEmote(AlphaPlayingEmote:double=1) -> (Output_Get:double)
   Set SimlationSpace(execute=@n12.else)
   > Set AlphaPlayingEmote(AlphaPlayingEmote=0.45)
  ]}
]
variables [
 AlphaPlayingEmote {type=number}
 Custom Idle Is Playing Float {type=number}
 Gravity Override {type=Vector}
 is_Jumping {type=number}
 OnlyCrouching {type=number}
 RBAN Settings Out {type=FortRigidBodyAnimNodeParameters}
 Sim Space Settings {
  type=AnimPhysSimSpaceSettings
  default {maxAngularAcceleration=10000, maxAngularVelocity=10000, simSpaceAngularAlpha=0, externalAngularVelocity=<0 0 0>}
 }
 SimlationSpace {type=AnimPhysSimSpaceType}
 "SkyDiving " {type=number}
 Trunk String ON {type=boolean}
]
